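# Material function that hides SpendPlume biped eyelids by collapsing their vertices to the object position through World Position Offset.
# Blink, upper-lid-down and lower-lid-up animation-curve parameters are thresholded into toggles, with lerps allowing inversion for default visibility.
# Four VertexColorMask calls match per-lid vertex-color values and combine by Max into a hidden mask.
# A left-side mask comes from pre-skinned local X, passed through a vertex interpolator for pixel use.
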MaterialFunction /CosmeticCompanions/Assets/Biped/SpendPlume/Materials/MF_HideEyeLids
graph {nodes [
  Constant() {r=1}
  n1 = GetMaterialAttributes(None=FunctionInput() {inputType=FunctionInput_MaterialAttributes}) {attributeGetTypes [#1]}
  n2 = WorldPosition()
  n3 = Multiply(A=Subtract(A=ObjectPositionWS(), B=Add(A=@n2.XYZ, B=@n1.WorldPositionOffset)), B=NamedRerouteUsage())
  FunctionOutput(None=SetMaterialAttributes!(MaterialAttributes=@n1.MaterialAttributes, WorldPositionOffset=Add(A=@n1.WorldPositionOffset, B=@n3)) {attributeSetTypes [#1]})
  n4 = ScalarParameter() {defaultValue=0.2, parameterName=L_UpperLid_Value}
  Multiply(A=@n4, B=OneMinus(None=NamedRerouteUsage()))
  n5 = ScalarParameter() {defaultValue=0.1, parameterName=R_UpperLid_Value}
  Multiply(A=@n5, B=NamedRerouteUsage())
  NamedRerouteDeclaration(None=ChannelMaskParameter(None=VertexColor()) {parameterName=EyelidMask}.R) {name=EyelidsMask, nodeColor=<rgba 0 0.2031 1 1>}
  NamedRerouteDeclaration(None=@n4) {name=L_UpperLid_Value, nodeColor=<rgba 1 0 0.8438 1>}
  n6 = ScalarParameter() {group=AnimationCurves, parameterName=R_lower_lid_up_pose}
  n7 = NamedRerouteUsage()
  NamedRerouteDeclaration(None=Ceil(None=Subtract(A=@n6, B=@n7) {constB=0.15})) {name=R_lower_lid_up_pose, nodeColor=<rgba 1 0 0.1406 1>}
  n8 = NamedRerouteUsage()
  n9 = Max(A=@n8, B=NamedRerouteUsage())
  n10 = OneMinus(None=@n9)
  n11 = NamedRerouteUsage()
  n12 = Max(A=NamedRerouteUsage(), B=NamedRerouteUsage())
  n13 = OneMinus(None=@n12)
  n14 = NamedRerouteUsage()
  n15 = Max(
   A=MaterialFunctionCall!(
    Hide Element Toggle=LinearInterpolate(A=@n9, B=@n10, Alpha=@n11)
    VertexColorChannel=NamedRerouteUsage()
    Vertex Value to hide=NamedRerouteUsage()
   ) {materialFunction=/Game/Packages/Fortress_SharedMaterials/Base_Material_Functions/VertexColorMask}.Result
   B=MaterialFunctionCall(
    Hide Element Toggle=LinearInterpolate(A=@n12, B=@n13, Alpha=@n14)
    VertexColorChannel=NamedRerouteUsage()
    Vertex Value to hide=NamedRerouteUsage()
   ) {materialFunction=&VertexColorMask}.Result
  )
  n16 = Max(A=NamedRerouteUsage(), B=NamedRerouteUsage())
  n17 = MaterialFunctionCall(
   Hide Element Toggle=LinearInterpolate(A=@n16, B=OneMinus(None=@n16), Alpha=NamedRerouteUsage())
   VertexColorChannel=NamedRerouteUsage()
   Vertex Value to hide=NamedRerouteUsage()
  ) {materialFunction=&VertexColorMask}
  n18 = Max(A=NamedRerouteUsage(), B=NamedRerouteUsage())
  n19 = LinearInterpolate(A=@n18, B=OneMinus(None=@n18), Alpha=NamedRerouteUsage())
  n20 = NamedRerouteUsage()
  n21 = NamedRerouteUsage()
  NamedRerouteDeclaration(None=Max(
    A=Max(A=@n15, B=@n17.Result)
    B=MaterialFunctionCall(Hide Element Toggle=@n19, VertexColorChannel=@n20, Vertex Value to hide=@n21) {materialFunction=&VertexColorMask}.Result
   )) {name=HiddenMask, nodeColor=<rgba 0.7813 0 1 1>}
  NamedRerouteDeclaration(None=ScalarParameter() {defaultValue=0.3, parameterName=Pose_Curve_Threshold}) {name=Threshold, nodeColor=<rgba 0.071 1 0.7063 1>}
  NamedRerouteDeclaration(None=ScalarParameter() {defaultValue=1, parameterName=DefaultHidden}) {name=DefaultHidden, nodeColor=<rgba 0.1016 1 0 1>}
  n22 = LocalPosition() {localOrigin=InstancePreSkinning}
  n23 = Saturate(None=Step(Y=ComponentMask(None=@n22.XYZ) {r=true}) {constX=0})
  NamedRerouteDeclaration(None=ShaderStageSwitch(PixelShader=VertexInterpolator(VS=@n23).PS, VertexShader=@n23)) {name=LeftMask, nodeColor=<rgba 0.1953 1 0 1>}
  NamedRerouteDeclaration(None=ScalarParameter() {defaultValue=0.25, parameterName=L_LowerLid_Value}) {name=L_LowerLid_Value, nodeColor=<rgba 1 0.307 0.3332 1>}
  NamedRerouteDeclaration(None=@n5) {name=R_UpperLid_Value, nodeColor=<rgba 0 0.4242 1 1>}
  NamedRerouteDeclaration(None=ScalarParameter() {defaultValue=0.15, parameterName=R_LowerLid_Value}) {name=R_LowerLid_Value, nodeColor=<rgba 0.071 1 0.7063 1>}
  n24 = ScalarParameter() {group=AnimationCurves, parameterName=L_blink_pose}
  n25 = NamedRerouteUsage()
  NamedRerouteDeclaration(None=Ceil(None=Subtract(A=@n24, B=@n25) {constB=0.15})) {name=L_blink_pose, nodeColor=<rgba 1 0 0.5391 1>}
  n26 = ScalarParameter() {group=AnimationCurves, parameterName=R_blink_pose}
  n27 = NamedRerouteUsage()
  NamedRerouteDeclaration(None=Ceil(None=Subtract(A=@n26, B=@n27) {constB=0.15})) {name=R_blink_pose, nodeColor=<rgba 0 0.8594 1 1>}
  n28 = ScalarParameter() {group=AnimationCurves, parameterName=L_upperLidDown_pose}
  n29 = NamedRerouteUsage()
  NamedRerouteDeclaration(None=Ceil(None=Subtract(A=@n28, B=@n29) {constB=0.15})) {name=L_upperLidDown_pose, nodeColor=<rgba 1 0.8203 0 1>}
  n30 = ScalarParameter() {group=AnimationCurves, parameterName=L_lower_lid_up_pose}
  n31 = NamedRerouteUsage()
  NamedRerouteDeclaration(None=Ceil(None=Subtract(A=@n30, B=@n31) {constB=0.15})) {name=L_lower_lid_up_pose, nodeColor=<rgba 0.5 0 1 1>}
  n32 = ScalarParameter() {group=AnimationCurves, parameterName=R_upperLidDown_pose}
  n33 = NamedRerouteUsage()
  NamedRerouteDeclaration(None=Ceil(None=Subtract(A=@n32, B=@n33) {constB=0.15})) {name=R_upperLidDown_pose, nodeColor=<rgba 0 1 0.1797 1>}
  OneMinus(None=@n8)
 ]}
