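# Animation Blueprint for the SpendPlume biped cosmetic companion, driving its skeleton by IK-retargeting another mesh's pose.
# Each update sets IsEmote when the owner is a FortCompanionMannequin or the IsEmote curve is positive.
# IsEmote eases between the standard retargeter and an emote retargeter whose lip and upper-cheek bones are overlaid with a pose copied from the attached parent.
# Several leftover nodes are unconnected.
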AnimBlueprint /CosmeticCompanions/Assets/Biped/SpendPlume/Meshes/SpendPlume_Biped_AnimBP
target_skeleton=&/SpendPlume_Biped_Skeleton
graphs [
 AnimGraph {nodes [
   Layered blend per bone() {
    curveBlendOption=NormalizeByWeight
    layerSetup [{branchFilters [
       {boneName=lip_lower_01_l}
       {boneName=lip_lower_01_r}
       {boneName=lip_lower_02_l}
       {boneName=lip_lower_02_r}
       {boneName=lip_lower_03_l}
       {boneName=lip_lower_03_r}
       {boneName=lip_lower_mid}
       {boneName=lip_corner_l}
       {boneName=lip_corner_r}
       {boneName=cheek_tp_in_l}
       {boneName=cheek_tp_in_r}
       {boneName=cheek_tp_ot_l}
       {boneName=cheek_tp_ot_r}
       {boneName=lip_upper_01_l}
       {boneName=lip_upper_01_r}
       {boneName=lip_upper_02_l}
       {boneName=lip_upper_02_r}
       {boneName=lip_upper_03_l}
       {boneName=lip_upper_03_r}
       {boneName=lip_upper_mid}
       {boneName=None}
       {boneName=None}
       {boneName=None}
       {boneName=None}
       {boneName=None}
       {boneName=None}
       {boneName=None}
       {boneName=None}
       {boneName=None}
       {boneName=None}
       {boneName=None}
       {boneName=None}
       {boneName=None}
       {boneName=None}
      ]}]
   }
   Layered blend per bone() {
    curveBlendOption=NormalizeByWeight
    layerSetup [{branchFilters [
       {boneName=lip_lower_01_l}
       {boneName=lip_lower_01_r}
       {boneName=lip_lower_02_l}
       {boneName=lip_lower_02_r}
       {boneName=lip_lower_03_l}
       {boneName=lip_lower_03_r}
       {boneName=lip_lower_mid}
       {boneName=lip_corner_l}
       {boneName=lip_corner_r}
       {boneName=cheek_tp_in_l}
       {boneName=cheek_tp_in_r}
       {boneName=cheek_tp_ot_l}
       {boneName=cheek_tp_ot_r}
       {boneName=lip_upper_01_l}
       {boneName=lip_upper_01_r}
       {boneName=lip_upper_02_l}
       {boneName=lip_upper_02_r}
       {boneName=lip_upper_03_l}
       {boneName=lip_upper_03_r}
       {boneName=lip_upper_mid}
       {boneName=None}
       {boneName=cheek_bt_l}
       {boneName=cheek_bt_r}
       {boneName=tongue_01}
       {boneName=tongue_02}
       {boneName=tongue_03}
       {boneName=None}
       {boneName=None}
       {boneName=None}
       {boneName=None}
       {boneName=None}
       {boneName=None}
       {boneName=None}
       {boneName=None}
      ]}]
   }
   Local Space Ref Pose()
   n1 = Layered blend per bone(
    BasePose=Retarget Pose From Mesh!() -> (Pose:Pose Link Structure) {is=title AnimGraphNode_RetargetPoseFromMesh, iKRetargeterAsset=&/RTG_SpendPlume_Emote}
    BlendPoses_0=Copy Pose From Mesh() {bCopyCurves=true, bUseAttachedParent=true}
   ) {
    curveBlendOption=NormalizeByWeight
    layerSetup [{branchFilters [
       {boneName=lip_lower_01_l}
       {boneName=lip_lower_01_r}
       {boneName=lip_lower_02_l}
       {boneName=lip_lower_02_r}
       {boneName=lip_lower_03_l}
       {boneName=lip_lower_03_r}
       {boneName=lip_lower_mid}
       {boneName=lip_corner_l}
       {boneName=lip_corner_r}
       {boneName=cheek_tp_in_l}
       {boneName=cheek_tp_in_r}
       {boneName=cheek_tp_ot_l}
       {boneName=cheek_tp_ot_r}
       {boneName=lip_upper_01_l}
       {boneName=lip_upper_01_r}
       {boneName=lip_upper_02_l}
       {boneName=lip_upper_02_r}
       {boneName=lip_upper_03_l}
       {boneName=lip_upper_03_r}
       {boneName=lip_upper_mid}
       {boneName=None}
       {boneName=None}
       {boneName=None}
       {boneName=None}
       {boneName=None}
       {boneName=None}
       {boneName=None}
       {boneName=None}
       {boneName=None}
       {boneName=None}
       {boneName=None}
       {boneName=None}
       {boneName=None}
       {boneName=None}
      ]}]
   }
   Output Pose(Result=Blend Poses by bool(
     BlendPose_0=@n1
     BlendPose_1=Retarget Pose From Mesh() {iKRetargeterAsset=&/RTG_SpendPlume}
     BlendTime_0=0.2
     BlendTime_1=0.2
     bActiveValue=Get IsEmote() -> (IsEmote:bool)
    ) {blendType=CubicInOut})
  ]}
 EventGraph {nodes [
   Event BlueprintUpdateAnimation()
   n1 = GetOwningActor()
   n2 = GetCurveValue(CurveName=IsEmote)
   > n3 = Branch(Condition=OR Boolean(
     A=IsA!(Object:Object Reference=@n1, ObjectClass:Object Class Reference=/Script/CosmeticCompanionsRuntime.FortCompanionMannequin) -> (ReturnValue:bool)
     B="float > float"(A=@n2)
    ))
   > Set IsEmote(IsEmote:bool=true) -> (Output_Get:bool)
   Set IsEmote(execute=@n3.else, IsEmote=false)
   TryGetPawnOwner()
   Make Retarget Profile!(
    RetargetOpProfiles:Array of Retarget Op Profile Structures
    bApplyTargetRetargetPose:bool=False
    TargetRetargetPoseName:Name=None
    bApplySourceRetargetPose:bool=False
    SourceRetargetPoseName:Name=None
    bForceAllIKOff:bool=False
    bApplyChainSettings:bool=False
    ChainSettings:Map of Names to Target Chain Settings Structures
    bApplyRootSettings:bool=False
    RootSettings:Target Root Settings Structure=(RotationAlpha=1,TranslationAlpha=1,BlendToSource=0,BlendToSourceWeights=(X=1,Y=1,Z=1),ScaleHorizontal=1,ScaleVertical=1,TranslationOffset=(X=0,Y=0,Z=0),RotationOffset=(Pitch=0,Yaw=0,Roll=0),AffectIKHorizontal=1,AffectIKVertical=0)
    bApplyGlobalSettings:bool=False
    GlobalSettings:Retarget Global Settings Structure=(bEnableRoot=True,bEnableFK=True,bEnableIK=True,bEnablePost=True,bCopyBasePose=False,CopyBasePoseRoot="",SourceScaleFactor=1,bWarping=False,DirectionSource=Goals,ForwardDirection=Y,DirectionChain="",WarpForwards=1,SidewaysOffset=0,WarpSplay=1)
   ) -> (RetargetProfile:Retarget Profile Structure) {is=title K2Node_MakeStruct}
  ]}
]
variables [IsEmote {type=boolean}]
